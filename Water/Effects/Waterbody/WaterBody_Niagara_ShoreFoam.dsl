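; Shore foam Blueprint for a Water Body, linking a Niagara grid simulation to a spline-mesh shoreline strip.
; Construction builds edge-offset segments along the water spline, vertex-painted by index, sharing a Shoreline MID fed the WaterZone info texture.
; Draw bakes offset spline positions as colors into a render target, debug-draws them, and creates the Niagara grid's float target.
; Draw and Copy Grid bind both targets to the MID.
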(event Custom|Draw
  (bind _display_mid (Variables|Default|GetDisplay%20MID))
  (bind _shore_rt_positions (Variables|Default|GetShore%20RT%20Positions))
  (Utilities|FlowControl|Sequence
    (:then_0
      (Rendering|ClearRenderTarget2D _shore_rt_positions)
      (bind (_canvas _size _context) (Rendering|BeginDrawCanvastoRenderTarget _shore_rt_positions))
      (Utilities|IsValid _canvas
        (:Is%20Valid
          (for _index (range (Variables|Default|GetSamples))
            (Canvas|DrawBox _canvas (Math|Vector2D|MakeVector2D (+ (Math|Conversions|ToFloat(Integer) _index) 0.5) 0.5) "(X=0.500000,Y=64.000000)" 1.0 (Math|Conversions|ToLinearColor(Vector) (+ (Spline|GetLocationatDistanceAlongSpline (Water|GetWaterSpline (Variables|Default|GetWater%20Body) :!class /Script/Water.WaterBody) (* (/ (Math|Conversions|ToFloat(Integer) _index) (Math|Conversions|ToFloat(Integer) (Variables|Default|GetSamples))) (Spline|GetSplineLength (Water|GetWaterSpline (Variables|Default|GetWater%20Body) :!class /Script/Water.WaterBody))) "World") (* (* (.Y (Math|Rotator|GetAxes (.rotation (Spline|GetTransformatDistanceAlongSpline (Water|GetWaterSpline (Variables|Default|GetWater%20Body) :!class /Script/Water.WaterBody) (* (/ (Math|Conversions|ToFloat(Integer) _index) (Math|Conversions|ToFloat(Integer) (Variables|Default|GetSamples))) (Spline|GetSplineLength (Water|GetWaterSpline (Variables|Default|GetWater%20Body) :!class /Script/Water.WaterBody))) "World" true)))) -1.0) (Variables|Default|GetEdge%20Bias))))))
          (Rendering|EndDrawCanvastoRenderTarget _context))))
    (:then_1
      (Rendering|Debug|FlushPersistentDebugLines)
      (for _index_2 (range (Variables|Default|GetSamples))
        (Rendering|Debug|DrawDebugSphere (Math|Conversions|ToVector(LinearColor) (Rendering|RenderTargetSampleUVEditorOnly (Variables|Default|GetShore%20RT%20Positions) (Math|Vector2D|MakeVector2D (/ (+ (Math|Conversions|ToFloat(Integer) _index_2) 0.5) (Math|Conversions|ToFloat(Integer) (Variables|Default|GetSamples))) 0.5))) 1e+02 12 "(R=1.000000,G=1.000000,B=1.000000,A=1.000000)" 1.0))
      (Utilities|IsValid (Variables|Default|GetNiagara)
        (:Is%20Valid
          (Niagara|SetTextureObject (Variables|/Script/Niagara.NiagaraActor|GetNiagaraComponent (Variables|Default|GetNiagara)) (Variables|Default|GetGridName) (Variables|Default|GetNiagara%20Grid%20Result))
          (Rendering|CreateAndAssignRenderTarget (Variables|Default|GetNiagara%20Grid%20Result) (Variables|Default|GetTexture%20Resolution) (Variables|Default|GetTexture%20Resolution) "RTF_R32f"
            (:Out
              (Rendering|Material|SetTextureParameterValue _display_mid "NiagaraRT" (Variables|Default|GetNiagara%20Grid%20Result))
              (Rendering|Material|SetTextureParameterValue _display_mid "SplineRT" (Variables|Default|GetShore%20RT%20Positions)))))))))

(event Custom|Copy%20Grid
  (bind _display_mid (Variables|Default|GetDisplay%20MID))
  (Rendering|Material|SetTextureParameterValue _display_mid "NiagaraRT" (Variables|Default|GetNiagara%20Grid%20Result))
  (Rendering|Material|SetTextureParameterValue _display_mid "SplineRT" (Variables|Default|GetShore%20RT%20Positions)))

(fn ConstructionScript
  (bind _water_body (Variables|Default|GetWater%20Body))
  (bind _returnvalue (Water|GetWaterSpline _water_body :!class /Script/Water.WaterBody))
  (bind _shore_scale (Variables|Default|GetShore%20Scale))
  (bind _returnvalue_2 (Math|Vector2D|MakeVector2D _shore_scale 1.0))
  (bind _spline_length (Variables|Default|GetSpline%20Length))
  (bind _returnvalue_3 (Spline|GetSplineLength _returnvalue))
  (bind _returnvalue_4 (/ _returnvalue_3 _spline_length))
  (bind _returnvalue_5 (Math|Float|Truncate _returnvalue_4))
  (Rendering|CreateAndAssignMID (Variables|Default|GetDisplay%20MID) "/Water/Effects/Niagara/Shoreline/Materials/Shoreline.Shoreline"
    (:Out
      (Rendering|Debug|FlushPersistentDebugLines)
      (Utilities|IsValid (Variables|Default|GetWater%20Body)
        (:Is%20Valid
          (bind _output_get (Variables|Default|SetSegments _returnvalue_5))
          (for _index (range _output_get)
            (bind _returnvalue_6 (AddComponent|NODE_AddSplineMeshComponent-1))
            (Rendering|Material|SetMaterial _returnvalue_6 0 (Variables|Default|GetDisplay%20MID))
            (SplineMesh|SetStartandEnd _returnvalue_6 (+ (.location (Spline|GetTransformatDistanceAlongSpline (Water|GetWaterSpline (Variables|Default|GetWater%20Body) :!class /Script/Water.WaterBody) (* (Math|Conversions|ToFloat(Integer) _index) (/ _returnvalue_3 (Math|Conversions|ToFloat(Integer) _returnvalue_5))) "World" true)) (* (* (.Y (Math|Rotator|GetAxes (.rotation (Spline|GetTransformatDistanceAlongSpline (Water|GetWaterSpline (Variables|Default|GetWater%20Body) :!class /Script/Water.WaterBody) (* (Math|Conversions|ToFloat(Integer) _index) (/ _returnvalue_3 (Math|Conversions|ToFloat(Integer) _returnvalue_5))) "World" true)))) -1.0) (Variables|Default|GetEdge%20Bias))) (/ (- (.location (Spline|GetTransformatDistanceAlongSpline (Water|GetWaterSpline (Variables|Default|GetWater%20Body) :!class /Script/Water.WaterBody) (Math|Float|%25%28Float%29 (* (+ (Math|Conversions|ToFloat(Integer) _index) 1.0) (/ _returnvalue_3 (Math|Conversions|ToFloat(Integer) _returnvalue_5))) _returnvalue_3) "World" true)) (.location (Spline|GetTransformatDistanceAlongSpline (Water|GetWaterSpline (Variables|Default|GetWater%20Body) :!class /Script/Water.WaterBody) (Math|Float|SelectFloat (- _returnvalue_3 (/ _returnvalue_3 (Math|Conversions|ToFloat(Integer) _returnvalue_5))) (* (+ (Math|Conversions|ToFloat(Integer) _index) -1.0) (/ _returnvalue_3 (Math|Conversions|ToFloat(Integer) _returnvalue_5))) (== _index 0)) "World" true))) 2.0) (+ (.location (Spline|GetTransformatDistanceAlongSpline (Water|GetWaterSpline (Variables|Default|GetWater%20Body) :!class /Script/Water.WaterBody) (Math|Float|%25%28Float%29 (* (+ (Math|Conversions|ToFloat(Integer) _index) 1.0) (/ _returnvalue_3 (Math|Conversions|ToFloat(Integer) _returnvalue_5))) _returnvalue_3) "World" true)) (* (* (.Y (Math|Rotator|GetAxes (.rotation (Spline|GetTransformatDistanceAlongSpline (Water|GetWaterSpline (Variables|Default|GetWater%20Body) :!class /Script/Water.WaterBody) (Math|Float|%25%28Float%29 (* (+ (Math|Conversions|ToFloat(Integer) _index) 1.0) (/ _returnvalue_3 (Math|Conversions|ToFloat(Integer) _returnvalue_5))) _returnvalue_3) "World" true)))) -1.0) (Variables|Default|GetEdge%20Bias))) (/ (- (.location (Spline|GetTransformatDistanceAlongSpline (Water|GetWaterSpline (Variables|Default|GetWater%20Body) :!class /Script/Water.WaterBody) (Math|Float|%25%28Float%29 (* (+ (+ (Math|Conversions|ToFloat(Integer) _index) 1.0) 1.0) (/ _returnvalue_3 (Math|Conversions|ToFloat(Integer) _returnvalue_5))) _returnvalue_3) "World" true)) (.location (Spline|GetTransformatDistanceAlongSpline (Water|GetWaterSpline (Variables|Default|GetWater%20Body) :!class /Script/Water.WaterBody) (* (Math|Conversions|ToFloat(Integer) _index) (/ _returnvalue_3 (Math|Conversions|ToFloat(Integer) _returnvalue_5))) "World" true))) 2.0) true :!class /Script/Engine.SplineMeshComponent)
            (SplineMesh|SetEndScale _returnvalue_6 _returnvalue_2 true :!class /Script/Engine.SplineMeshComponent)
            (SplineMesh|SetStartScale _returnvalue_6 _returnvalue_2 true :!class /Script/Engine.SplineMeshComponent)
            (VertexPaint|PaintVerticesSingleColor _returnvalue_6 (Utilities|Struct|MakeLinearColor (/ (Math|Conversions|ToFloat(Integer) _index) 255.0)) false))
          (bind _returnvalue_7 (Actor|GetActorOfClass "/Script/Water.WaterZone"))
          (Rendering|Material|SetTextureParameterValue (Variables|Default|GetDisplay%20MID) "WaterVelocityAndHeight" (Variables|/Script/Water.WaterZone|GetWaterInfoTextureArray _returnvalue_7)))))))
